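# Master material for glider cosmetics, a Default Lit surface built on material attributes and flagged for skeletal meshes and mesh deformers.
# All logic lives in a linear chain of material functions: glider texture inputs feed the Build-A-Brella layer, then the glider emissive function.
# The result passes through the shared weapon skin FX function and ends in the TechArt overrides function for accessories.
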Material /BRCosmetics/Gadgets/Assets/VinderTech_GliderChute/Materials/Master/M_FN_Glider_MASTER
graph {
 properties {
  bUseMaterialAttributes=true
  bUsedWithMeshDeformer=true
  bUsedWithSkeletalMesh=true
  pixelDepthOffsetMode=PDOM_Legacy
  usedShadingModels=Default Lit
 }
 nodes [
  n1 = MaterialFunctionCall!(In=MaterialFunctionCall() {materialFunction=/Game/Gadgets/Assets/VinderTech_GliderChute/Materials/Functions/MF_Glider_TextureInputs}.Result) {desc=Build-A-Brella, materialFunction=/Game/Gadgets/Assets/VinderTech_GliderChute/Materials/Functions/MF_Glider_BAB}
  MaterialFunctionCall(In=MaterialFunctionCall!(
    Pre-Skinned World Normal
    Gmap
    Customization Mask
    IsCharacter
    IsStatsB
    Material Attributes=MaterialFunctionCall!(Material=@n1.Result) {materialFunction=/Game/Gadgets/Assets/VinderTech_GliderChute/Materials/Functions/MF_Glider_Emissive}.Result
   ) {materialFunction=/Game/Weapons/WeaponSkins/MaterialFunction/Weapon_SkinFXFunction}.Result) {materialFunction=&MF_TechArtOverrides_accessories}
 ]
}
